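# Idle effect for the female Primal Falcon pickaxe that sheds drifting mesh particles from a single "feather" emitter.
# Particles spawn by random rate and movement distance in a torus offset from the SwingFX socket, then tumble under drag, gravity, curl noise and decaying inherited movement.
# User.IsSwing zeroes spawn rate and mesh scale and shrinks the movement threshold, suppressing feathers during swings.
# Front-end modifiers retune lifetime, velocity, drag and noise.
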NiagaraSystem /BRCosmetics/Weapons/FORT_Melee/Pickaxe_Primal_Falcon_Female/FX/NS_Pickaxe_Primal_Falcon_Female_Idle
dependencies {
 usedDataInterfaces [&NiagaraDataInterfaceSkeletalMesh, &NiagaraDataInterfaceCurve]
 usedDynamicInputs []
 usedModules [&SystemState]
}
emitters {feather {
  emitterUpdateScript [
   &EmitterState(Inactive Response=::NewEnumerator1, Life Cycle Mode=::NewEnumerator1, Loop Delay=0.05, Loop Duration=1, UseLoopDelay=true)
   &SpawnPerUnit(
    Max Movement Threshold:NiagaraFloat=&MakeCustomFloatFromBool(Bool=$User.IsSwing, True Float=0.1, False Float=2000)
    Module.Use Max Movement Threshold=true
    Module.Use Movement Tolerance=false
    Spawn Spacing=80
    Use Spawn Probability=false
   )
   &SpawnRate(
    SpawnRate=&Multiply_Float(
     A=&RandomRangeFloat(Minimum=0, Maximum=3)
     B=&MakeCustomFloatFromBool(Bool=$User.IsSwing, True Float=0, False Float=1)
    )
    Use Spawn Probability=false
   )
  ]
  particleSpawnScript [
   &InitializeParticle(
    Lifetime Max=&FN_FrontEndFloatModifier(Gameplay=0.8, Front End=0.9)
    Lifetime Min=0.7
    Lifetime Mode=::NewEnumerator1
    Mesh Scale Mode=::NewEnumerator1
    Mesh Uniform Scale Max=1.2
    Mesh Uniform Scale Min=0.8
    UsePositionOffset=false
    Write Material Random=true
   )
   &InitialMeshOrientation(Rotation=&RandomRangeVector(Minimum=<0 0 0>, Maximum=<0 0 0>), Use Rotation Vector=false)
   &AddVelocity(
    Scale Added Velocity=&VectorFromFloat(Value=&RandomRangeFloat(Minimum=0.5, Maximum=0.7))
    Velocity=&FN_FrontEndVector3Modifier(Gameplay=<0 -50 20>, Front End=<0 -70 20>)
   )
   &AddVelocityFromPoint(Module.Use Falloff Distance=false, Origin Offset=<0 10 0>, Velocity Strength=50) {enabled=false}
   &AddVelocityInCone(Cone Angle=30, Cone Axis=<0 0 1>, Module.Use Velocity Falloff On Cone Axis=false, Velocity Strength=100) {enabled=false}
   &SkeletalMeshLocation(
    Interpolate Mesh Sampling=false
    MaskSpawnGroup=false
    Mesh Sampling Type=::NewEnumerator5
    Sampled Position Offset=&Multiply_VectorByFloat(Vector=&FN_FrontEndVector3Modifier(Gameplay=<0 -39 0>, Front End=<0 -39 0>), Float=1)
    Skeletal Mesh=DataInterface() {propertyValues {
      DefaultMesh=null
      ExcludeBoneName=root
      OptionalComponentClass=null
      PreviewMesh=/BRCosmetics/Weapons/FORT_Melee/Pickaxe_Primal_Falcon_Female/Meshes/Primal_Falcon_Female_Axe
      SkinningMode=SkinOnTheFly
      SoftSourceActor=null
      SourceMode=Default
      UvSetIndex=0
      WholeMeshLOD=-1
      bExcludeBone=false
      bReadDeformedGeometry=true
      bRequireCurrentFrameData=true
      ComponentTags []
      FilteredBones []
      FilteredSockets [SwingFX]
      MeshUserParameter {parameter {name=None, typeDefHandle {registeredTypeIndex=1}, varData []}}
      SamplingRegions []
     }}
    Socket Sampling Mode=::NewEnumerator5
   )
   &SphereLocation(
    MaskSpawnGroup=false
    Override Seed=false
    Sphere Radius=15
    Surface Only=false
    UseNonUniformScale=false
    UseOffset=false
    UseSphereOrientation=false
   ) {enabled=false}
   &TorusLocation(
    Align Torus With Axis=true
    Handle Radius=0
    Large Radius=9
    MaskSpawnGroup=false
    Module.Override Randomness=false
    Normalized Angle Around Torus Axis=1
    Torus Axis=<0 1 -0.5>
   )
   &CameraOffset(Camera Offset Amount=3)
  ]
  particleUpdateScript [
   &Drag(
    Drag=&FN_FrontEndFloatModifier(Gameplay=1.5, Front End=1)
    Use Linear Drag=true
    Use Rotational Drag=false
   )
   &GravityForce(Gravity=<0 0 -50>)
   &InheritSourceMovement(Applied Position Scale=<0.3 0.3 0.3>, ApplyForce=false, ApplyPosition=true, ApplyPositionFalloff=false) {enabled=false}
   &InheritSourceMovement(Applied Position Scale=<0.3 0.3 0.3>, ApplyForce=false, ApplyPosition=true, ApplyPositionFalloff=true, ApplyPositionFalloffCurve=true) {module=InheritSourceMovement001}
   &CurlNoiseForce(
    Noise Strength=&FN_FrontEndFloatModifier(Gameplay=50, Front End=0)
    Pan Noise=true
    Pan Noise Field=<0 0 1>
   )
   &SolveForcesAndVelocity(Clamp Velocity=false, Limit Acceleration=false)
   &ParticleState()
   &DynamicMaterialParameters(
    Index 0 Param 1=&RandomRangeFloat(Minimum=0, Maximum=0.1)
    Index 0 Param 2=&FloatFromCurve(FloatCurve=DataInterface() {propertyValues {Curve {keys [<at 0.75 1>,<at 1 0>]}}})
    Index 0 Param 4=&RandomRangeFloat(Minimum=0, Maximum=0.5)
    Module.Param0WriteEnabled=true
    Module.Param1WriteEnabled=true
    Module.Param2WriteEnabled=true
    Module.Param3WriteEnabled=true
   )
   &ScaleColor(Scale RGB=&Multiply_VectorByFloat(Vector=<1 1 1>, Float=1.3), ScaleA=true, ScaleRGB=true)
   &UpdateMeshOrientation(
    Rotation Rate=&RandomRangeFloat(Minimum=0.4, Maximum=0.6)
    Rotation Vector=&VectorFromFloat(Value=&RandomRangeFloat(Minimum=-1, Maximum=1))
   )
   &ScaleMeshSize(Scale Factor=&Multiply_VectorByFloat(
     Vector=<1 1 1>
     Float=&FloatFromCurve(
      FloatCurve=DataInterface() {propertyValues {Curve {keys [
          {arriveTangent=4.1785, interpMode=RCIM_Cubic, leaveTangent=4.1785, tangentMode=RCTM_User, time=0, value=0}
          {arriveTangent=0.9995, interpMode=RCIM_Cubic, leaveTangent=-0.1395, tangentMode=RCTM_Break, time=0.15, value=1}
          {arriveTangent=-1.1765, interpMode=RCIM_Cubic, leaveTangent=-1.1765, time=0.5, value=1}
          {interpMode=RCIM_Cubic, leaveTangent=-1.5153, time=1, value=0}
         ]}}}
      Scale Curve=&MakeCustomFloatFromBool(Bool=$User.IsSwing, True Float=0, False Float=1)
     )
    ))
  ]
  rendererClasses [&NiagaraMeshRendererProperties]
  renderers [{rendererClass=&NiagaraMeshRendererProperties, rendererIndex=0}]
 }}
summary {
 systemName=NS_Pickaxe_Primal_Falcon_Female_Idle
 emitters [{emitterName=feather, rendererClasses [&NiagaraMeshRendererProperties]}]
 userVariables [
  User.bIsFrontEnd {type=NiagaraBool}
  User.bIsFrontEndPreview {type=NiagaraBool}
  User.IsSwing {type=NiagaraBool}
 ]
}
