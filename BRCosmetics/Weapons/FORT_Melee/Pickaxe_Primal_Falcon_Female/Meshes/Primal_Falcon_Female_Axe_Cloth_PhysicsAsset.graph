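# Physics asset for the Primal Falcon Female pickaxe, providing capsule bodies and joints for its dangling cloth bones.
# The handle and a two-capsule top_base form the weapon body, with left and right MSTR bodies attached to top_base by free-rotation constraints.
# A four-link left chain and a two-link right bottom chain hang from these, using limited swing and almost no twist.
# The right top chain branches from the first right bottom link rather than its own master.
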PhysicsAsset /BRCosmetics/Weapons/FORT_Melee/Pickaxe_Primal_Falcon_Female/Meshes/Primal_Falcon_Female_Axe_Cloth_PhysicsAsset
bodies {
 Melee_TwoHanded_Handle [{shapeType=Capsule, length=130.8806, radius=3.5558, center=<0.0004 -2.669 12.4562>, rotation=<rot 0 0 0.0002>}]
 dyn_cloth_l_01 [{shapeType=Capsule, length=5, radius=2.25, shapeName=dyn_cloth_l_01_capsule, center=<5 0 0>, rotation=<rot 90 0 0>}]
 dyn_cloth_l_02 [{shapeType=Capsule, length=5, radius=2.25, shapeName=dyn_cloth_l_02_capsule, center=<5 0 0>, rotation=<rot 90 0 0>}]
 dyn_cloth_l_03 [{shapeType=Capsule, length=5, radius=2.25, shapeName=dyn_cloth_l_03_capsule, center=<5 0 0>, rotation=<rot 90 0 0>}]
 dyn_cloth_l_04 [{shapeType=Capsule, length=5, radius=2.25, shapeName=dyn_cloth_l_04_capsule, center=<5 0 0>, rotation=<rot 90 0 0>}]
 dyn_cloth_l_MSTR [{shapeType=Capsule, length=1, radius=1, shapeName=dyn_cloth_l_MSTR_capsule, rotation=<rot 0 -90 0>}]
 dyn_cloth_r_bot_01 [{shapeType=Capsule, length=5, radius=2, shapeName=dyn_cloth_r_bot_01_capsule, center=<4.0968 0 0>, rotation=<rot 90 0 0>}]
 dyn_cloth_r_bot_02 [{shapeType=Capsule, length=5, radius=2, shapeName=dyn_cloth_r_bot_02_capsule, center=<1.2842 0 0>, rotation=<rot 90 0 0>}]
 dyn_cloth_r_bot_MSTR [{shapeType=Capsule, length=1, radius=1, shapeName=dyn_cloth_r_bot_MSTR_capsule, rotation=<rot 90 0 0>}]
 dyn_cloth_r_top_01 [{shapeType=Capsule, length=5, radius=2, shapeName=dyn_cloth_r_top_01_capsule, rotation=<rot 90 0 0>}]
 dyn_cloth_r_top_02 [{shapeType=Capsule, length=5, radius=2, shapeName=dyn_cloth_r_top_02_capsule, rotation=<rot 90 0 0>}]
 top_base [
  {shapeType=Capsule, length=28, radius=3, shapeName=top_base_capsule, center=<-5.1988 -13.3593 0>, rotation=<rot 0 0 89.9999>}
  {
   shapeType=Capsule
   length=24.5
   radius=5.5
   shapeName=top_base_capsule
   center=<-1.531 -16.5284 -14.7522>
   rotation=<rot -59.9995 0 -10.0008>
  }
 ]
}
constraints [
 {bone1Name=dyn_cloth_r_bot_01, bone2Name=dyn_cloth_r_bot_MSTR, swing1LimitDegrees=30, swing2LimitDegrees=45, twistLimitDegrees=0.01}
 {bone1Name=dyn_cloth_r_bot_02, bone2Name=dyn_cloth_r_bot_01, swing1LimitDegrees=30, swing2LimitDegrees=45, twistLimitDegrees=0.01}
 {bone1Name=dyn_cloth_r_top_01, bone2Name=dyn_cloth_r_bot_01, swing1LimitDegrees=30, swing2LimitDegrees=45, twistLimitDegrees=0.01}
 {bone1Name=dyn_cloth_r_top_02, bone2Name=dyn_cloth_r_top_01, swing1LimitDegrees=30, swing2LimitDegrees=45, twistLimitDegrees=0.01}
 {
  bone1Name=top_base
  bone2Name=Melee_TwoHanded_Handle
  swing1LimitDegrees=45
  swing1Motion=Free
  swing2LimitDegrees=45
  swing2Motion=Free
  twistLimitDegrees=45
  twistMotion=Free
 }
 {bone1Name=dyn_cloth_l_01, bone2Name=dyn_cloth_l_MSTR, swing1LimitDegrees=30, swing2LimitDegrees=45, twistLimitDegrees=0.01}
 {bone1Name=dyn_cloth_l_02, bone2Name=dyn_cloth_l_01, swing1LimitDegrees=30, swing2LimitDegrees=45, twistLimitDegrees=0.01}
 {bone1Name=dyn_cloth_l_03, bone2Name=dyn_cloth_l_02, swing1LimitDegrees=30, swing2LimitDegrees=45, twistLimitDegrees=0.01}
 {bone1Name=dyn_cloth_l_04, bone2Name=dyn_cloth_l_03, swing1LimitDegrees=30, swing2LimitDegrees=45, twistLimitDegrees=0.01}
 {
  bone1Name=dyn_cloth_l_MSTR
  bone2Name=top_base
  swing1LimitDegrees=45
  swing1Motion=Free
  swing2LimitDegrees=45
  swing2Motion=Free
  twistLimitDegrees=45
  twistMotion=Free
 }
 {
  bone1Name=dyn_cloth_r_bot_MSTR
  bone2Name=top_base
  swing1LimitDegrees=45
  swing1Motion=Free
  swing2LimitDegrees=45
  swing2Motion=Free
  twistLimitDegrees=45
  twistMotion=Free
 }
]
