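# Physics asset for the feathers on the female Primal Falcon pickaxe, defining their bodies and joints for secondary motion.
# One long capsule covers the two-handed handle, and small capsules cover each dynamic feather bone.
# Nine feather chains of two to five bones attach to the handle, each bone constrained to the one before it.
# Tight swing and twist limits on every joint keep the feathers close to their modeled shape.
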PhysicsAsset /BRCosmetics/Weapons/FORT_Melee/Pickaxe_Primal_Falcon_Female/Meshes/Primal_Falcon_Female_Axe_Feathers_PhysicsAsset
bodies {
 Melee_TwoHanded_Handle [{shapeType=Capsule, length=129.3806, radius=5.0558, center=<0.0003 -1.0559 12.4562>, rotation=<rot 0 0 0.0002>}]
 dyn_feather_l_bk_bot_01 [{shapeType=Capsule, length=5, radius=2, shapeName=dyn_feather_l_bk_bot_01_capsule, center=<1.7251 0 0>, rotation=<rot 90 0 0>}]
 dyn_feather_l_bk_bot_02 [{shapeType=Capsule, length=5, radius=2, shapeName=dyn_feather_l_bk_bot_02_capsule, center=<1.7251 0 0>, rotation=<rot 90 0 0>}]
 dyn_feather_l_bk_mid_01 [{shapeType=Capsule, length=5, radius=2, shapeName=dyn_feather_l_bk_mid_01_capsule, center=<1.7251 0 0>, rotation=<rot 90 0 0>}]
 dyn_feather_l_bk_mid_02 [{shapeType=Capsule, length=5, radius=2, shapeName=dyn_feather_l_bk_mid_02_capsule, center=<1.7251 0 0>, rotation=<rot 90 0 0>}]
 dyn_feather_l_bk_top_01 [{shapeType=Capsule, length=5, radius=2, shapeName=dyn_feather_l_bk_top_01_capsule, center=<1.7251 0 0>, rotation=<rot 90 0 0>}]
 dyn_feather_l_bk_top_02 [{shapeType=Capsule, length=5, radius=2, shapeName=dyn_feather_l_bk_top_02_capsule, center=<1.7251 0 0>, rotation=<rot 90 0 0>}]
 dyn_feather_l_bk_top_03 [{shapeType=Capsule, length=5, radius=2, shapeName=dyn_feather_l_bk_top_03_capsule, center=<1.7251 0 0>, rotation=<rot 90 0 0>}]
 dyn_feather_l_top_01 [{shapeType=Capsule, length=5, radius=2, shapeName=dyn_feather_l_top_01_capsule, center=<1.7251 0 0>, rotation=<rot 90 0 0>}]
 dyn_feather_l_top_02 [{shapeType=Capsule, length=5, radius=2, shapeName=dyn_feather_l_top_02_capsule, center=<1.7251 0 0>, rotation=<rot 90 0 0>}]
 dyn_feather_l_top_03 [{shapeType=Capsule, length=5, radius=2, shapeName=dyn_feather_l_top_03_capsule, center=<1.7251 0 0>, rotation=<rot 90 0 0>}]
 dyn_feather_l_top_04 [{shapeType=Capsule, length=5, radius=2, shapeName=dyn_feather_l_top_04_capsule, center=<1.7251 0 0>, rotation=<rot 90 0 0>}]
 dyn_feather_l_top_05 [{shapeType=Capsule, length=5, radius=2, shapeName=dyn_feather_l_top_05_capsule, center=<1.7251 0 0>, rotation=<rot 90 0 0>}]
 dyn_feather_mid_bot_01 [{shapeType=Capsule, length=5, radius=2, shapeName=dyn_feather_mid_bot_01_capsule, center=<1.7251 0 0>, rotation=<rot 90 0 0>}]
 dyn_feather_mid_bot_02 [{shapeType=Capsule, length=5, radius=2, shapeName=dyn_feather_mid_bot_02_capsule, center=<1.7251 0 0>, rotation=<rot 90 0 0>}]
 dyn_feather_mid_bot_03 [{shapeType=Capsule, length=5, radius=2, shapeName=dyn_feather_mid_bot_03_capsule, center=<1.7251 0 0>, rotation=<rot 90 0 0>}]
 dyn_feather_mid_top_01 [{shapeType=Capsule, length=5, radius=2, shapeName=dyn_feather_mid_top_01_capsule, center=<1.7251 0 0>, rotation=<rot 90 0 0>}]
 dyn_feather_mid_top_02 [{shapeType=Capsule, length=5, radius=2, shapeName=dyn_feather_mid_top_02_capsule, center=<1.7251 0 0>, rotation=<rot 90 0 0>}]
 dyn_feather_mid_top_03 [{shapeType=Capsule, length=5, radius=2, shapeName=dyn_feather_mid_top_03_capsule, center=<1.7251 0 0>, rotation=<rot 90 0 0>}]
 dyn_feather_r_bk_bot_01 [{shapeType=Capsule, length=5, radius=2, shapeName=dyn_feather_r_bk_bot_01_capsule, center=<1.7251 0 0>, rotation=<rot 90 0 0>}]
 dyn_feather_r_bk_bot_02 [{shapeType=Capsule, length=5.75, radius=1.25, shapeName=dyn_feather_r_bk_bot_02_capsule, center=<1.7251 0 0>, rotation=<rot 90 0 0>}]
 dyn_feather_r_bk_bot_03 [{shapeType=Capsule, length=5.75, radius=1.25, shapeName=dyn_feather_r_bk_bot_03_capsule, center=<1.7251 0 0>, rotation=<rot 90 0 0>}]
 dyn_feather_r_bk_bot_04 [{shapeType=Capsule, length=5.75, radius=1.25, shapeName=dyn_feather_r_bk_bot_04_capsule, center=<1.7251 0 0>, rotation=<rot 90 0 0>}]
 dyn_feather_r_bk_mid_01 [{shapeType=Capsule, length=5, radius=2, shapeName=dyn_feather_r_bk_mid_01_capsule, center=<1.7251 0 0>, rotation=<rot 90 0 0>}]
 dyn_feather_r_bk_mid_02 [{shapeType=Capsule, length=5, radius=2, shapeName=dyn_feather_r_bk_mid_02_capsule, center=<1.7251 0 0>, rotation=<rot 90 0 0>}]
 dyn_feather_r_bk_mid_03 [{shapeType=Capsule, length=5, radius=2, shapeName=dyn_feather_r_bk_mid_03_capsule, center=<1.7251 0 0>, rotation=<rot 90 0 0>}]
 dyn_feather_r_top_01 [{shapeType=Capsule, length=5, radius=2, shapeName=dyn_feather_r_top_01_capsule, center=<1.7251 0 0>, rotation=<rot 90 0 0>}]
 dyn_feather_r_top_02 [{shapeType=Capsule, length=5, radius=2, shapeName=dyn_feather_r_top_02_capsule, center=<1.7251 0 0>, rotation=<rot 90 0 0>}]
 dyn_feather_r_top_03 [{shapeType=Capsule, length=5, radius=2, shapeName=dyn_feather_r_top_03_capsule, center=<1.7251 0 0>, rotation=<rot 90 0 0>}]
}
constraints [
 {bone1Name=dyn_feather_l_top_01, bone2Name=Melee_TwoHanded_Handle, swing1LimitDegrees=5, swing2LimitDegrees=5, twistLimitDegrees=5}
 {bone1Name=dyn_feather_l_top_02, bone2Name=dyn_feather_l_top_01, swing1LimitDegrees=5, swing2LimitDegrees=8, twistLimitDegrees=2}
 {bone1Name=dyn_feather_l_top_03, bone2Name=dyn_feather_l_top_02, swing1LimitDegrees=5, swing2LimitDegrees=8, twistLimitDegrees=2}
 {bone1Name=dyn_feather_l_top_04, bone2Name=dyn_feather_l_top_03, swing1LimitDegrees=5, swing2LimitDegrees=8, twistLimitDegrees=2}
 {bone1Name=dyn_feather_l_top_05, bone2Name=dyn_feather_l_top_04, swing1LimitDegrees=5, swing2LimitDegrees=8, twistLimitDegrees=2}
 {bone1Name=dyn_feather_l_bk_mid_01, bone2Name=Melee_TwoHanded_Handle, swing1LimitDegrees=5, swing2LimitDegrees=8, twistLimitDegrees=2}
 {bone1Name=dyn_feather_l_bk_mid_02, bone2Name=dyn_feather_l_bk_mid_01, swing1LimitDegrees=5, swing2LimitDegrees=8, twistLimitDegrees=2}
 {bone1Name=dyn_feather_l_bk_bot_01, bone2Name=Melee_TwoHanded_Handle, swing1LimitDegrees=5, swing2LimitDegrees=8, twistLimitDegrees=2}
 {bone1Name=dyn_feather_l_bk_bot_02, bone2Name=dyn_feather_l_bk_bot_01, swing1LimitDegrees=5, swing2LimitDegrees=8, twistLimitDegrees=2}
 {bone1Name=dyn_feather_r_bk_mid_01, bone2Name=Melee_TwoHanded_Handle, swing1LimitDegrees=5, swing2LimitDegrees=8, twistLimitDegrees=2}
 {bone1Name=dyn_feather_r_bk_mid_02, bone2Name=dyn_feather_r_bk_mid_01, swing1LimitDegrees=5, swing2LimitDegrees=8, twistLimitDegrees=2}
 {bone1Name=dyn_feather_r_bk_mid_03, bone2Name=dyn_feather_r_bk_mid_02, swing1LimitDegrees=5, swing2LimitDegrees=8, twistLimitDegrees=2}
 {bone1Name=dyn_feather_r_bk_bot_01, bone2Name=Melee_TwoHanded_Handle, swing1LimitDegrees=5, swing2LimitDegrees=8, twistLimitDegrees=2}
 {bone1Name=dyn_feather_r_bk_bot_02, bone2Name=dyn_feather_r_bk_bot_01, swing1LimitDegrees=5, swing2LimitDegrees=8, twistLimitDegrees=2}
 {bone1Name=dyn_feather_r_bk_bot_03, bone2Name=dyn_feather_r_bk_bot_02, swing1LimitDegrees=5, swing2LimitDegrees=8, twistLimitDegrees=2}
 {bone1Name=dyn_feather_r_bk_bot_04, bone2Name=dyn_feather_r_bk_bot_03, swing1LimitDegrees=5, swing2LimitDegrees=8, twistLimitDegrees=2}
 {bone1Name=dyn_feather_r_top_01, bone2Name=Melee_TwoHanded_Handle, swing1LimitDegrees=5, swing2LimitDegrees=8, twistLimitDegrees=2}
 {bone1Name=dyn_feather_r_top_02, bone2Name=dyn_feather_r_top_01, swing1LimitDegrees=5, swing2LimitDegrees=8, twistLimitDegrees=2}
 {bone1Name=dyn_feather_r_top_03, bone2Name=dyn_feather_r_top_02, swing1LimitDegrees=5, swing2LimitDegrees=8, twistLimitDegrees=2}
 {bone1Name=dyn_feather_mid_bot_01, bone2Name=Melee_TwoHanded_Handle, swing1LimitDegrees=5, swing2LimitDegrees=8, twistLimitDegrees=2}
 {bone1Name=dyn_feather_mid_bot_02, bone2Name=dyn_feather_mid_bot_01, swing1LimitDegrees=5, swing2LimitDegrees=8, twistLimitDegrees=2}
 {bone1Name=dyn_feather_mid_bot_03, bone2Name=dyn_feather_mid_bot_02, swing1LimitDegrees=5, swing2LimitDegrees=8, twistLimitDegrees=2}
 {bone1Name=dyn_feather_mid_top_01, bone2Name=Melee_TwoHanded_Handle, swing1LimitDegrees=5, swing2LimitDegrees=8, twistLimitDegrees=2}
 {bone1Name=dyn_feather_mid_top_02, bone2Name=dyn_feather_mid_top_01, swing1LimitDegrees=5, swing2LimitDegrees=8, twistLimitDegrees=2}
 {bone1Name=dyn_feather_mid_top_03, bone2Name=dyn_feather_mid_top_02, swing1LimitDegrees=5, swing2LimitDegrees=8, twistLimitDegrees=2}
 {bone1Name=dyn_feather_l_bk_top_01, bone2Name=Melee_TwoHanded_Handle, swing1LimitDegrees=2, swing2LimitDegrees=2, twistLimitDegrees=2}
 {bone1Name=dyn_feather_l_bk_top_03, bone2Name=dyn_feather_l_bk_top_02, swing1LimitDegrees=5, swing2LimitDegrees=8, twistLimitDegrees=2}
 {bone1Name=dyn_feather_l_bk_top_02, bone2Name=dyn_feather_l_bk_top_01, swing1LimitDegrees=5, swing2LimitDegrees=5, twistLimitDegrees=2}
]
