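# Physics asset for the tassel on the Primal Falcon Female two-handed pickaxe, defining collision and joints for its strap bones.
# A long capsule covers the handle, and three small capsules form a strap chain, each constrained to the previous bone with the first anchored to the handle.
# Joints allow moderate swing, wider on one axis than the other, and almost no twist, so the strap bends without spinning along its length.
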PhysicsAsset /BRCosmetics/Weapons/FORT_Melee/Pickaxe_Primal_Falcon_Female/Meshes/Primal_Falcon_Female_Axe_Tassel_PhysicsAsset
bodies {
 Melee_TwoHanded_Handle [{shapeType=Capsule, length=100.1306, radius=2.3058, center=<0.8417 -2.6691 12.4562>, rotation=<rot 0 0 0.0002>}]
 dyn_strap_01 [{shapeType=Capsule, length=2.5, radius=0.5, shapeName=dyn_strap_01_capsule, center=<3.857 0 -0.6952>, rotation=<rot 90 0 0>}]
 dyn_strap_02 [{shapeType=Capsule, length=2.5, radius=0.5, shapeName=dyn_strap_02_capsule, center=<3.7694 0 -0.6953>, rotation=<rot 90 0 0>}]
 dyn_strap_03 [{shapeType=Capsule, length=2.25, radius=0.75, shapeName=dyn_strap_03_capsule, center=<3.2191 0 -0.6953>, rotation=<rot 90 0 0>}]
}
constraints [
 {bone1Name=dyn_strap_01, bone2Name=Melee_TwoHanded_Handle, swing1LimitDegrees=30, swing2LimitDegrees=5, twistLimitDegrees=0.1}
 {bone1Name=dyn_strap_02, bone2Name=dyn_strap_01, swing1LimitDegrees=30, swing2LimitDegrees=15, twistLimitDegrees=0.1}
 {bone1Name=dyn_strap_03, bone2Name=dyn_strap_02, swing1LimitDegrees=25, swing2LimitDegrees=10, twistLimitDegrees=0.1}
]
